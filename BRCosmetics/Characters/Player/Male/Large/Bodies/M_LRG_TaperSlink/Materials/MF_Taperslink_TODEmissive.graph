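# Material function for the TaperSlink character body, adding time-of-day glow-in-the-dark emissive and blinking red/green lights to incoming material attributes.
# Masked glow follows morning, day, evening and night values, with a front-end looping override and eye-adaptation compensation.
# A stepped sine alternates red and green LEDs, split by local X and masked by the TA input, over the emissive.
# EmoteSwitch inverts that split and blends in a hue-shifted base color.
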MaterialFunction /BRCosmetics/Characters/Player/Male/Large/Bodies/M_LRG_TaperSlink/Materials/MF_Taperslink_TODEmissive
graph {nodes [
  n1 = GetMaterialAttributes(None=FunctionInput() {inputName=MA, inputType=FunctionInput_MaterialAttributes}) {attributeGetTypes [#1, #2]}
  n2 = NamedRerouteUsage()
  n3 = NamedRerouteUsage()
  n4 = Step(Y=MaterialFunctionCall(Speed=ScalarParameter() {defaultValue=1.2, group=BlinkingLights, parameterName=BlinkEmissiveSpeed}) {materialFunction=&DebugTimeSine}.Result) {constX=0.5}
  n5 = Multiply(A=Multiply(A=Saturate(None=@n4), B=Saturate(None=@n3)), B=Constant3Vector() {constant=<rgba 1 0 0 1>})
  n6 = NamedRerouteUsage()
  n7 = OneMinus(None=@n4)
  n8 = Multiply(A=Multiply(A=Saturate(None=@n7), B=Saturate(None=@n6)), B=Constant3Vector() {constant=<rgba 0 1 0 1>})
  n9 = Multiply(
   A=Subtract(A=Add(A=@n5, B=@n8), B=Fresnel() {baseReflectFraction=0.1, exponent=1.5})
   B=ScalarParameter() {defaultValue=10, group=BlinkingLights, parameterName=BlinkEmissiveStrength}
  ) {constB=50}
  n10 = FunctionInput() {inputName=TA_BlinkLightsMask}
  n11 = NamedRerouteDeclaration(None=ComponentMask(None=@n10) {b=true}) {name=TABlueMask, nodeColor=<rgba 0 1 0.6953 1>}
  n12 = ComponentMask(None=@n10) {r=true}
  n13 = Multiply(A=ScalarParameter() {group=BlinkingLights, parameterName=BlinkLightsEmissive}, B=Saturate(None=Subtract(A=@n11, B=@n12)))
  n14 = NamedRerouteUsage()
  n15 = NamedRerouteUsage()
  n16 = NamedRerouteUsage()
  n17 = LinearInterpolate(
   A=NamedRerouteUsage()
   B=MaterialFunctionCall(Hue Shift Percentage=ScalarParameter() {group=BlinkingLights, parameterName="Hue Shift Percentage (S)"}) {materialFunction=&HueShift}.Result
   Alpha=Saturate(None=Multiply(A=@n15, B=@n16))
  )
  n18 = ScalarParameter() {group=BlinkingLights, parameterName=EmoteSwitch}
  FunctionOutput(None=SetMaterialAttributes!(
    MaterialAttributes=@n1.MaterialAttributes
    EmissiveColor=LinearInterpolate(A=@n2, B=@n9, Alpha=@n13)
    BaseColor=LinearInterpolate(A=@n14, B=@n17, Alpha=@n18)
   ) {attributeSetTypes [#1, #2]})
  n19 = LocalPosition()
  n20 = ComponentMask(None=LinearInterpolate(A=@n19.XYZ, B=OneMinus(None=@n19.XYZ), Alpha=ScalarParameter() {group=BlinkingLights, parameterName=EmoteSwitch})) {r=true}
  NamedRerouteDeclaration(None=OneMinus(None=@n20)) {name=RedLedMask, nodeColor=<rgba 1 0 0 1>}
  NamedRerouteDeclaration(None=@n20) {name=GreenLedMask, nodeColor=<rgba 0 1 0 1>}
  n21 = TextureObjectParameter() {
   group=TOD_Emissive
   parameterName=GlowDarkMask
   samplerType=SAMPLERTYPE_Masks
   texture=/Game/Athena/Aircraft/PartyBus_RT/Textures/AllBlackMasks
  }
  n22 = TextureSample(Tex=@n21) {samplerType=SAMPLERTYPE_Masks, texture=/BRCosmetics/Characters/Player/Male/Large/Bodies/M_LRG_TaperSlink/Textures/T_TaperSlink_Body_FX1}
  n23 = ScalarParameter() {defaultValue=1, group=TOD_Emissive, parameterName=GlowDarkEmissiveStrength}
  n24 = VectorParameter() {group=TOD_Emissive, parameterName=GlowDarkColor, defaultValue=<rgba 0.0126 0.8348 0.0145 1>}
  n25 = Multiply(
   A=MaterialFunctionCall(
    Value_Morning=ScalarParameter() {defaultValue=0.2, group=TOD_Emissive, parameterName=Value_Morning}
    Value_Day=ScalarParameter() {group=TOD_Emissive, parameterName=Value_Day}
    Value_Evening=ScalarParameter() {defaultValue=0.3, group=TOD_Emissive, parameterName=Value_Evening}
    Value_Night=ScalarParameter() {defaultValue=0.8, group=TOD_Emissive, parameterName=Value_Night}
   ) {materialFunction=&Scalar_Time_of_Day}.Result
   B=@n24.RGB
  )
  n26 = Multiply(B=@n24.RGB) {constA=0.6}
  n27 = Multiply(A=ScalarParameter() {group=TOD_Emissive, parameterName=IsFrontEnd}, B=NamedRerouteUsage())
  n28 = LinearInterpolate(
   A=NamedRerouteUsage()
   B=Multiply(A=Multiply(A=@n22.R, B=@n23) {constB=2}, B=LinearInterpolate(A=@n25, B=@n26, Alpha=@n27) {constA=0.2})
   Alpha=ScalarParameter() {defaultValue=1, group=TOD_Emissive, parameterName=GlowInDarkActivation}
  )
  n29 = EyeAdaptation()
  NamedRerouteDeclaration(None=Multiply(A=Divide(A=@n28, B=@n29.EyeAdaptation), B=ScalarParameter() {defaultValue=1, group=TOD_Emissive, parameterName=ReactivityToggleEmissive})) {name=TODEmissive, nodeColor=<rgba 0.3068 1 0.2373 1>}
  NamedRerouteDeclaration(None=@n1.EmissiveColor) {name=Emissive, nodeColor=<rgba 0.0547 0 1 1>}
  NamedRerouteDeclaration(None=@n1.BaseColor) {name=BaseColor, nodeColor=<rgba 0.2656 1 0 1>}
  n30 = Divide(A=Time(), B=ScalarParameter() {defaultValue=4, parameterName="FE Loop`Duration"}) {constB=5}
  n31 = Modulo!(A=@n30, B=Constant() {r=2})
  n32 = Frac(None=@n30)
  n33 = Power(Base=Abs(None=ConstantBiasScale(None=@n32) {bias=-0.5, scale=2})) {constExponent=3}
  NamedRerouteDeclaration(None=Multiply(A=Floor(None=@n31), B=OneMinus(None=@n33))) {name=FE Loop, nodeColor=<rgba 0 0.4141 1 1>}
 ]}
