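# Animation Blueprint for the TaperSlink large male body, following the parent mesh's pose and hiding parts of itself while its matching glider is open.
# The copied pose runs through a Control Rig, then constraints tie backpack and cape attach bones to spine_04.
# Customization load flags whether the TaperSlink glider is equipped; updates then toggle "Hide Element" material parameters, partly after a delay.
# A hidden right arm lowers the rig's gauntlet alpha.
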AnimBlueprint /BRCosmetics/Characters/Player/Male/Large/Bodies/M_LRG_TaperSlink/Meshes/M_LRG_TaperSlink_AnimBP
target_skeleton=&/M_LRG_TaperSlink_Skeleton
graphs [
 AnimGraph {nodes [
   n1 = Control Rig!(
    Source=Copy Pose From Mesh() {bUseAttachedParent=true}
    Alpha
    RightArm_is_HidenIsWearingGauntlet_alpha_R:double=Get IsWearingGauntlet_alpha_Right() -> (IsWearingGauntlet_alpha_Right:double)
   ) -> (Pose) {controlRigAssetReference {blueprintRigClass=&/M_LRG_TaperSlink_CtrlRig.M_LRG_TaperSlink_CtrlRig_C}}
   n2 = "Constraint - attach_backpack (P)"(ComponentPose=Local To Component(LocalPose=@n1), Alpha=1, ConstraintWeights_0=1) {boneToModify {boneName=attach_backpack}, constraintSetup [{targetBone {boneName=spine_04}}], constraintWeights [1]}
   n3 = "Constraint - attach_cape (P)"(ComponentPose=@n2, Alpha=1, ConstraintWeights_0=1) {boneToModify {boneName=attach_cape}, constraintSetup [{targetBone {boneName=spine_04}}], constraintWeights [1]}
   Output Pose(Result=Component To Local(ComponentPose=@n3)) {layerGroup=None}
  ]}
 EventGraph {nodes [
   Event BlueprintUpdateAnimation()
   > n1 = Sequence!(execute) -> (then_0, then_1, then_2:Exec)
   n2 = GetPartSkeletalMeshComponent()
   n3 = Get Dynamic Material Instance(execute=@n1.then_1, SkeletalMeshComponent=@n2, ElementIndex=1)
   n4 = AND Boolean(A=Get TaperSlinkGlider() -> (TaperSlinkGlider:bool), B=Get bIsParachuteOpened())
   > n5 = Branch(Condition=@n4)
   > SetScalarParameterValue(self=@n3, ParameterName=Hide Element 01, Value=1)
   SetScalarParameterValue(execute=@n5.else, self=@n3, ParameterName=Hide Element 01, Value=0)
   n6 = Get Dynamic Material Instance(execute=@n1.then_0, SkeletalMeshComponent=@n2)
   > n7 = Branch(Condition=@n4)
   > Delay(Duration=0.2)
   > SetScalarParameterValue(self=@n6, ParameterName=Hide Element 01, Value=1)
   > SetScalarParameterValue(self=@n6, ParameterName=Hide Element 05, Value=1)
   SetScalarParameterValue(execute=@n7.else, self=@n6, ParameterName=Hide Element 01, Value=0)
   > SetScalarParameterValue(self=@n6, ParameterName=Hide Element 05, Value=0)
   n8 = Branch(execute=@n1.then_2, Condition=Get bIsRightArmHidden() -> (bIsRightArmHidden:bool))
   > Set IsWearingGauntlet_alpha_Right(IsWearingGauntlet_alpha_Right:double=0.16) -> (Output_Get:double)
   Set IsWearingGauntlet_alpha_Right(execute=@n8.else, IsWearingGauntlet_alpha_Right=1)
   n9 = Event OnCharacterCustomizationLoaded()
   > Set SkelMesh(SkelMesh=@n9.CustomBaseMesh)
   > n10 = Cast To FortPlayerPawn(Object=GetOwningActor())
   > Is Valid(InputObject=Get TaperSlinkGliderSkelMesh() -> (TaperSlinkGliderSkelMesh:Skeletal Mesh Object Reference))
   > n11 = Set Current Glider(Current Glider:Athena Glider Item Definition Object Reference=GetItemDefinitionTypeFromLoadoutWithActor!(
     Actor:Actor Object Reference=@n10
     LoadoutSlot:Gameplay Tag Structure=(TagName="LoadoutSlot.Glider")
     CosmeticType:Athena Cosmetic Item Definition Class Reference=/Script/FortniteGame.AthenaGliderItemDefinition
    ) -> (OutCosmeticItemDef:Athena Glider Item Definition Object Reference)) -> (Output_Get:Athena Glider Item Definition Object Reference)
   > Set TaperSlinkGlider(TaperSlinkGlider:bool="Equal (Object)"(A=@n11, B=Get GliderItemDef() -> (GliderItemDef:Athena Glider Item Definition Object Reference)) {is=title K2Node_CallFunction}) -> (Output_Get:bool)
  ]}
]
variables [
 Array {type="array<integer>"}
 Array 0 {type="array<integer>"}
 Array_0 {type="array<integer>", default [0,1]}
 Current Glider {type=AthenaGliderItemDefinition, default=null}
 Element Index {type=integer, default=1}
 GliderItemDef {type=AthenaGliderItemDefinition, default=/BRCosmetics/Athena/Items/Cosmetics/Gliders/Glider_TaperSlink}
 IsWearingGauntlet_alpha_Left {type=number}
 IsWearingGauntlet_alpha_Right {type=number}
 NewVar {type=boolean}
 SkelMesh {type=SkeletalMeshComponent}
 TaperSlinkGlider {type=boolean}
 TaperSlinkGliderSkelMesh {type=SkeletalMesh, default=/BRCosmetics/Gadgets/Assets/VinderTech_GliderChute/Glider_TaperSlink/Meshes/TaperSlink_Glider}
]
